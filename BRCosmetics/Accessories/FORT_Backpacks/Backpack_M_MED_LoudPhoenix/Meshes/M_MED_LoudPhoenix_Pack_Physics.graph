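# Physics asset for the LoudPhoenix cosmetic backpack mesh, giving the pack simple box collision bodies.
# Two bodies, on the pack and dyn_pack_base bones, each use one thin box roughly covering a flat slab shape.
# A single constraint joins dyn_pack_base to pack with swing and twist locked, so the base cannot rotate relative to the pack despite the configured angle limits.
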PhysicsAsset /BRCosmetics/Accessories/FORT_Backpacks/Backpack_M_MED_LoudPhoenix/Meshes/M_MED_LoudPhoenix_Pack_Physics
bodies {
 dyn_pack_base [{shapeType=Box, extentX=25, extentY=4, extentZ=26, shapeName=dyn_pack_base_box, center=<0 -0.3161 -0.5551>}]
 pack [{shapeType=Box, extentX=24.5, extentY=2.5, extentZ=32, shapeName=pack_box, center=<0 1.9064 -1.4326>}]
}
constraints [{
  bone1Name=dyn_pack_base
  bone2Name=pack
  swing1LimitDegrees=45
  swing1Motion=Locked
  swing2LimitDegrees=45
  swing2Motion=Locked
  twistLimitDegrees=45
  twistMotion=Locked
 }]
